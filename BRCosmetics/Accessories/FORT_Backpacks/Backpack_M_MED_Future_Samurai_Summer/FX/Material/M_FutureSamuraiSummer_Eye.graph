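# Translucent unlit effect material for an animated, blinking eye on the Future Samurai Summer backpack cosmetic.
# Procedural bent wire stripes combine with sphere-masked eyeball rings and scrolling-capable textures on bump-offset UVs.
# A sine-timed blink, offset per particle or vertex color, masks that pattern, which drives a color ramp or optional curve atlas color.
# Edge darkening and time-of-day tint shape the emissive, alongside opacity-mask and normal-based vertex offset math.
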Material /BRCosmetics/Accessories/FORT_Backpacks/Backpack_M_MED_Future_Samurai_Summer/FX/Material/M_FutureSamuraiSummer_Eye
graph {
 properties {
  bUsedWithMeshDeformer=true
  bUsedWithNiagaraMeshParticles=true
  bUsedWithSkeletalMesh=true
  blendMode=BLEND_Translucent
  pixelDepthOffsetMode=PDOM_Legacy
  shadingModel=MSM_Unlit
  twoSided=true
  usedShadingModels=Unlit
 }
 nodes [
  n1 = TextureCoordinate() {coordinateIndex=1}
  n2 = MaterialFunctionCall(Float2=@n1) {materialFunction=&BreakOutFloat2Components}
  n3 = Frac(None=@n2.R)
  n4 = OneMinus(None=@n3)
  n5 = Ceil(None=Subtract(A=@n3) {constB=0.5})
  n6 = OneMinus(None=Multiply(A=LinearInterpolate(A=@n3, B=@n4, Alpha=@n5)) {constB=2})
  n7 = ScalarParameter() {defaultValue=1.8, parameterName=Bend_AreaPower}
  n8 = Add(A=Multiply(A=Power(Base=@n6, Exp=@n7) {constExponent=1.8}, B=ScalarParameter() {defaultValue=0.3629, parameterName=Bend_Strength}), B=@n1)
  n9 = ScalarParameter() {defaultValue=-0.68, parameterName=W_CentWidthScale_ScaleCenterOffset}
  n10 = Multiply(A=Add(A=@n8, B=@n9), B=ScalarParameter() {defaultValue=2.9498, parameterName=W_CentWidthScale_Scale})
  n11 = Multiply(A=@n9) {constB=-1}
  n12 = Sine(None=Multiply(A=@n2.R) {constB=0.5})
  n13 = Power(Base=@n12, Exp=ScalarParameter() {defaultValue=3, parameterName=W_CentWidthScale_AreaMaskPower})
  n14 = Power(Base=@n12, Exp=ScalarParameter() {defaultValue=6, parameterName=W_WireOffset_AreaMaskPower})
  n15 = Add(
   A=LinearInterpolate(A=Add(A=@n10, B=@n11), B=@n8, Alpha=OneMinus(None=@n13))
   B=LinearInterpolate(
    A=ScalarParameter() {defaultValue=0.092, parameterName=W_WireOffset_Center}
    B=ScalarParameter() {defaultValue=0.084, parameterName=W_WireOffset_Edge}
    Alpha=OneMinus(None=@n14)
   )
  )
  n16 = ScalarParameter() {defaultValue=4.9183, parameterName=W_TotalWireWidth}
  n17 = AppendVector(A=@n16, B=Constant() {r=1})
  n18 = Multiply(
   A=MaterialFunctionCall(Float2=MaterialFunctionCall(UVs=@n15, Texture Scale=@n17) {materialFunction=&ScaleUVsByCenter}.UVs) {materialFunction=&BreakOutFloat2Components}.G
   B=@n16
  )
  n19 = Frac(None=@n18)
  n20 = LinearInterpolate(A=Multiply(A=@n19) {constB=-1}, B=OneMinus(None=@n19), Alpha=Ceil(None=Subtract(A=@n19) {constB=0.5}))
  n21 = Frac(None=MaterialFunctionCall(Float2=TextureCoordinate()) {materialFunction=&BreakOutFloat2Components}.G)
  n22 = Subtract(A=@n21) {constB=0.5}
  n23 = Multiply(A=LinearInterpolate(A=@n21, Alpha=Ceil(None=@n22))) {constB=2}
  n24 = Multiply(
   A=ScalarParameter() {defaultValue=0.5, parameterName=W_EdgeStrecth_Strength}
   B=Power(Base=OneMinus(None=@n23), Exp=ScalarParameter() {defaultValue=2, parameterName=W_EdgeStretch_AreaPow})
  )
  n25 = Frac(None=Add(A=@n18, B=Multiply(A=@n20, B=@n24)))
  n26 = LinearInterpolate(A=@n25, B=OneMinus(None=@n25), Alpha=Ceil(None=Subtract(A=@n25) {constB=0.5}))
  n27 = MaterialFunctionCall(
   In=Add(A=Multiply(A=@n26) {constB=2}, B=ScalarParameter() {defaultValue=-0.15, parameterName=W_PerWireWidth})
   Contrast=ScalarParameter() {defaultValue=1, parameterName=W_PerWireEdgeSharp}
  ) {materialFunction=&CheapContrast}
  n28 = Constant() {r=1}
  n29 = Constant()
  n30 = If(A=@n18, B=ScalarParameter() {defaultValue=4.5, parameterName=W_WireMask_ClampHigh}, "A > B"=@n28, "A == B"=@n29, "A < B"=@n29)
  n31 = ScalarParameter() {defaultValue=1.5, parameterName=W_WireMask_ClampLow}
  n32 = ComponentMask(None=VectorParameter() {parameterName=EyeBall_Center, defaultValue=<rgba 0.5 0.27 0 1>}.RGB) {g=true, r=true}
  n33 = NamedRerouteUsage()
  n34 = MaterialFunctionCall(
   UVs=@n33
   Texture Scale=ComponentMask(None=VectorParameter() {parameterName=EyeBall_SphereMaskUVScale, defaultValue=<rgba 1 1 0 1>}.RGB) {g=true, r=true}
  ) {materialFunction=&ScaleUVsByCenter}
  n35 = ScalarParameter() {defaultValue=1, parameterName=EyeBall_Hardness}
  n36 = SphereMask(A=@n32, B=@n34.UVs, Radius=ScalarParameter() {defaultValue=0.2, parameterName=EyeBall_Size}, Hardness=@n35)
  n37 = VectorParameter() {parameterName=EyeBall_Out_TextureRemap, defaultValue=<rgba 0 0.1 0 0.1>}
  n38 = TextureObjectParameter() {
   group=EyeOutTexture
   parameterName=EyeOutTexture_Texture
   sortPriority=0
   texture=/Game/Global/Textures/Default/Tileable_Patterns/T_Thick_Smoke_05
  }
  n39 = VectorParameter() {group=EyeOutTexture, parameterName=EyeOutTexture_UVControl, channelNames {a=VPan, b=UPan, g=VScale, r=UScale}, defaultValue=<rgba 1 1 0 0>}
  n40 = MaterialFunctionCall(Texture=@n38, UVControl_RGB=@n39, UVControl_A=@n39, "[Override]UVInput"=NamedRerouteUsage()) {materialFunction=&FXCommon_TextureSample}
  n41 = ScalarParameter() {parameterName=EyeOutTexture_Power}
  n42 = Multiply(
   A=@n36
   B=MaterialFunctionCall(Input=Power(Base=@n40.R, Exp=@n41), Input Low=Constant(), Input High=Constant() {r=1}, Target Low=@n37.R, Target High=@n37.R) {materialFunction=&RemapValueRange}.Result
  )
  n43 = Saturate(None=SphereMask(
    A=@n32
    B=@n34.UVs
    Radius=ScalarParameter() {defaultValue=0.19, parameterName=EyeBall_OutEdgeWidth}
    Hardness=ScalarParameter() {defaultValue=1, parameterName=EyeBall_OutEdgeHardness}
   ))
  n44 = ScalarParameter() {defaultValue=0.1, parameterName=EyeBall_Inner_Size}
  n45 = Subtract(
   A=LinearInterpolate(A=LinearInterpolate(A=@n27.Result, Alpha=@n30), Alpha=If(A=@n18, B=@n31, "A > B"=@n29, "A == B"=@n28, "A < B"=@n28))
   B=LinearInterpolate(A=LinearInterpolate(A=@n36, B=@n42, Alpha=@n43) {constA=1}, Alpha=SphereMask(A=@n32, B=@n34.UVs, Radius=@n44, Hardness=@n35))
  )
  n46 = MaterialFunctionCall(Float2=TextureCoordinate()) {materialFunction=&BreakOutFloat2Components}
  n47 = DynamicParameter() {defaultValue=<rgba 0 1 1 1>, paramNames [Random,Param2,Param3,Param4]}
  n48 = Multiply(A=VertexColor().R) {constB=123.456}
  n49 = Add(A=Time(), B=If(A=@n47.Random, B=Constant(), "A > B"=@n47.Random, "A == B"=@n48, "A < B"=@n48))
  n50 = VectorParameter() {parameterName=BlinkTimeScale, defaultValue=<rgba 2 0.5 0 1>}
  n51 = Subtract(A=Sine(None=Multiply(A=@n49, B=@n50.G) {constB=0.5}), B=@n50.B)
  n52 = Multiply(A=Saturate(None=Ceil(None=@n51)), B=Sine(None=Multiply(A=@n49, B=@n50.R) {constB=2}))
  n53 = Saturate(None=Ceil(None=Add(A=@n46.G, B=@n52)))
  n54 = Saturate(None=LinearInterpolate(A=VectorParameter() {parameterName=BlinkColor, defaultValue=<rgba 0 0 0 1>}.RGB, B=@n45, Alpha=@n53))
  n55 = Constant2Vector() {g=0.5, r=0.5}
  n56 = VectorParameter() {parameterName=Opacity_SphereMaskUVScale, defaultValue=<rgba 3 1 0 1>}
  n57 = MaterialFunctionCall(UVs=@n33, Texture Scale=AppendVector(A=@n56.R, B=@n56.R)) {materialFunction=&ScaleUVsByCenter}
  n58 = Constant() {r=0.3}
  n59 = Constant() {r=0.25}
  n60 = Multiply(A=Power(Base=SphereMask(A=@n55, B=@n57.UVs, Radius=@n58, Hardness=@n59)) {constExponent=1}) {constB=1.5}
  n61 = TextureCoordinate()
  n62 = VectorParameter() {parameterName=Opacity_UVMask_MulAndPower, defaultValue=<rgba 10 0.3 0 1>}
  n63 = Saturate(None=Multiply(
    A=MaterialFunctionCall(Float2=@n61) {materialFunction=&BreakOutFloat2Components}.G
    B=@n62.R
   ) {constB=5})
  Add(A=@n54, B=Multiply(A=Saturate(None=@n60), B=Power(Base=@n63, Exp=@n62.G) {constExponent=1}))
  n64 = VectorParameter() {parameterName=Color_Morning, defaultValue=<rgba 0.1 0.1 0.1 1>}
  n65 = Constant() {r=1}
  n66 = VectorParameter() {parameterName=Color_Evening, defaultValue=<rgba 0.1 0.1 0.1 1>}
  n67 = VectorParameter() {parameterName=Color_Night, defaultValue=<rgba 0.4922 0.5 0.4844 1>}
  n68 = MaterialFunctionCall(Float2=TextureCoordinate()) {materialFunction=&BreakOutFloat2Components}
  n69 = Multiply(A=Min(A=@n68.G, B=OneMinus(None=@n68.G))) {constB=2}
  n70 = OneMinus(None=Power(Base=OneMinus(None=@n69), Exp=ScalarParameter() {defaultValue=4, parameterName=EdgeDarken_Power2}))
  n71 = VectorParameter() {group=Color, parameterName=DarkC, defaultValue=<rgba 0.007 0.0056 0.0056 1>}
  n72 = VectorParameter() {group=Color, parameterName=LightC, defaultValue=<rgba 1 1 1 1>}
  n73 = CurveAtlasRowParameter(CurveTime=ComponentMask(None=@n54) {r=true}) {
   atlas=/Game/Athena/LinearColorCurves/A_Atlas_Turtleneck_Glider
   curve=/Game/Characters/Player/Female/Medium/Bodies/F_MED_Cactus_Rocker/FX/Cactus_Rocker_TrailCurve
   defaultValue=6
   parameterName=EyeColorCurve
  }
  n74 = Min(A=@n68.R, B=OneMinus(None=@n68.R))
  n75 = ScalarParameter() {defaultValue=5, parameterName=EdgeDarken_Mul}
  n76 = Power(Base=Saturate(None=Multiply(A=@n74, B=@n75) {constB=2}), Exp=ScalarParameter() {defaultValue=2, parameterName=EdgeDarken_Power})
  n77 = TextureObjectParameter() {
   group=EyeEdgeTexture
   parameterName=EyeEdgeTexture_Texture
   sortPriority=0
   texture=&T_Thick_Smoke_05
  }
  n78 = VectorParameter() {
   group=EyeEdgeTexture
   parameterName=EyeEdgeTexture_UVControl
   channelNames {a=VPan, b=UPan, g=VScale, r=UScale}
   defaultValue=<rgba 0.7 1.5 0 0>
  }
  n79 = OneMinus(None=MaterialFunctionCall(Texture=@n77, UVControl_RGB=@n78, UVControl_A=@n78, "[Override]UVInput"=NamedRerouteUsage()) {materialFunction=&FXCommon_TextureSample}.R)
  n80 = VectorParameter() {group=EyeEdgeTexture, parameterName=EyeEdgeTexture_PowerAndMul, defaultValue=<rgba 1.5 -0.8 2 0>}
  n81 = Multiply(A=OneMinus(None=Power(Base=@n79, Exp=@n80.R) {constExponent=5}), B=@n80.G) {constB=0.15}
  n82 = Power(Base=Multiply(A=@n74) {constB=2}, Exp=@n80.B)
  n83 = LinearInterpolate(A=@n76, B=Add(A=@n76, B=LinearInterpolate(A=@n81, Alpha=@n82) {constB=0}), Alpha=@n76)
  n84 = Multiply(
   A=@n83
   B=LinearInterpolate(A=LinearInterpolate(A=@n71.RGB, B=@n72.RGB, Alpha=@n54), B=@n73, Alpha=ScalarParameter() {parameterName=UseCurveColor})
  )
  n85 = VectorParameter() {parameterName=EdgeDarken_ColorVer_DarkC, defaultValue=<rgba 0 0 0 1>}
  n86 = VectorParameter() {parameterName=EdgeDarken_ColorVer_LightC, defaultValue=<rgba 0 0 0 1>}
  n87 = Add(
   A=Multiply(A=OneMinus(None=@n83), B=VectorParameter() {parameterName=EdgeDarken_ColorVer_Add, defaultValue=<rgba 0 0 0 1>}.RGB)
   B=Multiply(A=LinearInterpolate(A=@n85.RGB, B=@n86.RGB, Alpha=@n83), B=@n84)
  )
  n88 = ScalarParameter() {parameterName=EdgeDarken_ColorVer}
  n89 = Multiply(A=@n70, B=Saturate(None=LinearInterpolate(A=@n84, B=@n87, Alpha=@n88)))
  n90 = Multiply(A=LinearInterpolate(A=@n85.RGB, B=@n86.RGB, Alpha=@n70), B=@n89)
  MaterialFunctionCall(Emissive=Multiply(
    A=MaterialFunctionCall(Color_Morning=@n64.RGB, Color_Day=@n65, Color_Evening=@n66.RGB, Color_Night=@n67.RGB) {materialFunction=&Color_Time_of_Day}.Result
    B=LinearInterpolate(A=@n89, B=@n90, Alpha=@n88)
   )) {materialFunction=&MF_FX_EmissiveAdjustments}
  MaterialFunctionCall(Float3=ObjectPositionWS()) {materialFunction=&BreakOutFloat3Components}
  n91 = Frac(None=MaterialFunctionCall(Float2=TextureCoordinate()) {materialFunction=&BreakOutFloat2Components}.G)
  n92 = Subtract(A=@n91) {constB=0.5}
  n93 = Multiply(A=LinearInterpolate(A=@n91, B=OneMinus(None=@n91), Alpha=Ceil(None=@n92))) {constB=2}
  n94 = OneMinus(None=Power(Base=OneMinus(None=@n93), Exp=ScalarParameter() {defaultValue=4, parameterName=WPO_EdgeAreaPow}))
  n95 = ScalarParameter() {defaultValue=0.8, parameterName=WPO_EdgeStrength}
  Multiply(A=VertexNormalWS(), B=Add(A=Multiply(A=@n94, B=@n95), B=ScalarParameter() {defaultValue=0.3, parameterName=WPO_GlobalOffset}))
  Multiply(A=@n53, B=@n45)
  NamedRerouteDeclaration(None=BumpOffset(
    Coordinate=TextureCoordinate()
    Height=ScalarParameter() {parameterName=UV0Height}
    HeightRatioInput=ScalarParameter() {defaultValue=0.05, parameterName=UV0HeightRatio}
   )) {name=BumpUV0, nodeColor=<rgba 0 0.0391 1 1>}
  NamedRerouteDeclaration(None=BumpOffset(
    Coordinate=TextureCoordinate() {coordinateIndex=1}
    Height=ScalarParameter() {parameterName=UV1Height}
    HeightRatioInput=ScalarParameter() {defaultValue=0.05, parameterName=UV1HeightRatio}
   )) {name=BumpUV1, nodeColor=<rgba 0 0.6928 1 1>}
  Time()
 ]
}
