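# Unlit skeletal-mesh effect material for the Future Samurai Summer backpack, named EyeEdge, drawing animated noise as emissive color.
# Two noise layers and an optional Fresnel term distort mesh or bounding-box UVs into the main texture, with an optional second texture.
# The gray-adjusted result can take a curve-atlas gradient and texture mask, then is either scaled or maxed with fake normal-mapped sky-light shading.
# Time-of-day color multiplies the output before shared emissive adjustments.
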Material /BRCosmetics/Accessories/FORT_Backpacks/Backpack_M_MED_Future_Samurai_Summer/FX/Material/M_FutureSamuraiSummer_EyeEdge
graph {
 properties {bUsedWithMeshDeformer=true, bUsedWithSkeletalMesh=true, pixelDepthOffsetMode=PDOM_Legacy, shadingModel=MSM_Unlit, usedShadingModels=Unlit}
 nodes [
  Constant() {r=1}
  n1 = PreSkinnedLocalBounds()
  n2 = MaterialFunctionCall(Float3=Divide(A=PreSkinnedPosition(), B=@n1.Extents)) {materialFunction=&BreakOutFloat3Components}
  n3 = AppendVector(A=@n2.G, B=@n2.G)
  Divide(A=@n3) {constB=1024}
  n4 = VectorParameter() {group=TOD, parameterName=Color_Morning, defaultValue=<rgba 0.1 0.1 0.1 1>}
  n5 = Constant() {r=1}
  n6 = VectorParameter() {group=TOD, parameterName=Color_Evening, defaultValue=<rgba 0.1 0.1 0.1 1>}
  n7 = VectorParameter() {group=TOD, parameterName=Color_Night, defaultValue=<rgba 0.4922 0.5 0.4844 1>}
  n8 = ScalarParameter() {parameterName=GradientColor_BBoxUVOffset_U}
  n9 = ScalarParameter() {parameterName=GradientColor_BBoxUVOffset_V}
  n10 = MaterialFunctionCall() {materialFunction=&BoundingBoxBased_0-1_UVW}
  n11 = StaticSwitchParameter(True=VertexInterpolator(VS=@n3).PS, False=AppendVector(A=@n10.G, B=@n10.G)) {parameterName=UsePreSkinData}
  n12 = VectorParameter() {parameterName=GradientColor_BBoxUVScaleAndRotate, defaultValue=<rgba 1 1 0 1>}
  n13 = Rotator(Coordinate=@n11, Time=@n12.B)
  n14 = AppendVector(A=@n12.R, B=@n12.R)
  n15 = TextureObjectParameter() {
   group=MainTexture
   parameterName=MainTexture_Texture
   samplerType=SAMPLERTYPE_Masks
   sortPriority=0
   texture=/Game/Weapons/WeaponSkins/Wraps/Textures/Mash/Mash_Noise
  }
  n16 = VectorParameter() {group=MainTexture, parameterName=MainTexture_UVControl, channelNames {a=VPan, b=UPan, g=VScale, r=UScale}, defaultValue=<rgba 2 0.25 0 0.06>}
  n17 = TextureObjectParameter() {group=NoiseA, parameterName=NoiseA_Texture, sortPriority=0, texture=/Game/Effects/Fort_Effects/Emotes/UltraLight/Texture/T_SoftNoise_BC7}
  n18 = VectorParameter() {group=NoiseA, parameterName=NoiseA_UVControl, channelNames {a=VPan, b=UPan, g=VScale, r=UScale}, defaultValue=<rgba 2.5 1 0 0.3>}
  n19 = VectorParameter() {parameterName=BBoxUVScaleAndRotate, defaultValue=<rgba 1 1 0 1>}
  n20 = Multiply(A=Rotator(Coordinate=@n11, Time=@n19.B), B=AppendVector(A=@n19.R, B=@n19.R))
  n21 = StaticSwitchParameter(True=@n20, False=TextureCoordinate()) {parameterName=UseBBOXUV}
  n22 = VectorParameter() {
   group=NoiseA
   parameterName=NoiseA_NoiseControl
   sortPriority=60
   channelNames {a=VStrength, b=UStrength, g=VOffset, r=UOffset}
   defaultValue=<rgba -0.5 -0.5 0.1 0>
  }
  n23 = MaterialFunctionCall!(
   NoiseValue=MaterialFunctionCall(Texture=@n17, UVControl_RGB=@n18, UVControl_A=@n18, "[Override]UVInput"=@n21) {materialFunction=&FXCommon_TextureSample}.RG
   Control_RGB=@n22
   Control_A=@n22
  ) {materialFunction=/Game/Effects/Niagara/MasterMaterials/Functions/Common/FXCommon_NoiseControl}
  n24 = TextureSampleParameter2D() {
   group=Lighting
   parameterName=Lighting_Normal
   samplerType=SAMPLERTYPE_Normal
   texture=&/T_M_MED_FutureSamuraiSummer_Backpack_N
  }
  n25 = Transform(None=@n24.RGB)
  n26 = Power(
   Base=Fresnel(ExponentIn=Constant() {r=1}, BaseReflectFractionIn=Constant(), Normal=@n25)
   Exp=ScalarParameter() {parameterName=FresnelNoise_Exponent}
  )
  n27 = ScalarParameter() {parameterName=FresnelNoise_Strength}
  n28 = TextureObjectParameter() {
   group=NoiseB
   parameterName=NoiseB_Texture
   samplerType=SAMPLERTYPE_LinearGrayscale
   sortPriority=0
   texture=/Game/Effects/Fort_Effects/Emotes/UltraLight/Texture/T_Cloud_Greyscale
  }
  n29 = VectorParameter() {group=NoiseB, parameterName=NoiseB_UVControl, channelNames {a=VPan, b=UPan, g=VScale, r=UScale}, defaultValue=<rgba 3 1 0 0.1>}
  n30 = MaterialFunctionCall(Texture=@n28, UVControl_RGB=@n29, UVControl_A=@n29, "[Override]UVInput"=@n21) {materialFunction=&FXCommon_TextureSample}
  n31 = VectorParameter() {
   group=NoiseB
   parameterName=NoiseB_NoiseControl
   sortPriority=60
   channelNames {a=VStrength, b=UStrength, g=VOffset, r=UOffset}
   defaultValue=<rgba -0.5 -0.5 0.05 0>
  }
  n32 = Add(
   A=Add(A=Add(A=@n23.NoiseValue, B=@n21), B=Multiply(A=@n26, B=@n27))
   B=MaterialFunctionCall(NoiseValue=AppendVector(A=@n30.R, B=@n30.R), Control_RGB=@n31, Control_A=@n31) {materialFunction=&FXCommon_NoiseControl}.NoiseValue
  )
  n33 = MaterialFunctionCall(Texture=@n15, UVControl_RGB=@n16, UVControl_A=@n16, "[Override]UVInput"=@n32) {materialFunction=&FXCommon_TextureSample}
  n34 = TextureObjectParameter() {group=MainTexture2, parameterName=MainTexture2_Texture, sortPriority=0}
  n35 = VectorParameter() {group=MainTexture2, parameterName=MainTexture2_UVControl, channelNames {a=VPan, b=UPan, g=VScale, r=UScale}, defaultValue=<rgba 1 1 0 0>}
  n36 = StaticSwitchParameter(
   True=Add(
    A=@n33.R
    B=MaterialFunctionCall(Texture=@n34, UVControl_RGB=@n35, UVControl_A=@n35, "[Override]UVInput"=@n32) {materialFunction=&FXCommon_TextureSample}.R
   )
   False=@n33.R
  ) {parameterName=UseSecTexture}
  n37 = Add(
   A=Add(A=AppendVector(A=@n8, B=@n9), B=Multiply(A=@n13, B=@n14))
   B=Multiply(A=@n36, B=ScalarParameter() {parameterName=Gradient_TextureOffset})
  )
  n38 = CurveAtlasRowParameter(CurveTime=ComponentMask(None=StaticSwitchParameter(True=@n37, False=@n20) {parameterName=UseGradientColor}) {r=true}) {
   atlas=/Game/Athena/LinearColorCurves/A_Atlas_Turtleneck_Glider
   curve=/Game/Weapons/FORT_Melee/Pickaxe_Future_Samurai_Summer_Male/FX/Red/Curve_Pickaxe_Future_Samurai_Summer
   defaultValue=28
   group=Color
   parameterName=Curve
  }
  n39 = ScalarParameter() {defaultValue=0.1, group=MainTexture, parameterName=MainTexture_GrayAdjust_Add}
  n40 = Saturate(None=Multiply(A=Add(A=@n36, B=@n39), B=ScalarParameter() {defaultValue=1, group=MainTexture, parameterName=MainTexture_GrayAdjust_Mul}))
  n41 = ScalarParameter() {defaultValue=0.33, group=MainTexture, parameterName=MainTexture_GrayAdjust_Power}
  n42 = VectorParameter() {group=MainTexture, parameterName="MainTexture_GrayAdjust_MulAfterPower(Color)", defaultValue=<rgba 0.02 0.018 0.018 1>}
  n43 = Multiply(A=Saturate(None=Power(Base=@n40, Exp=@n41)), B=@n42.RGB)
  n44 = StaticSwitchParameter(True=Multiply(A=@n38.R, B=@n43), False=@n43) {parameterName=UseGradientColor}
  n45 = TextureObjectParameter() {
   parameterName=Mask
   samplerType=SAMPLERTYPE_LinearGrayscale
   texture=&/T_M_MED_FutureSamuraiSummer_Backpack_Mask
  }
  n46 = TextureSample(Tex=@n45) {
   samplerType=SAMPLERTYPE_LinearGrayscale
   texture=&T_M_MED_FutureSamuraiSummer_Backpack_Mask
  }
  n47 = MaterialFunctionCall(X=@n46.R, Y=@n46.R, Z=@n46.R) {materialFunction=&MakeFloat3}
  n48 = ScalarParameter() {parameterName=MaskLowRemap}
  n49 = Constant() {r=1}
  n50 = StaticSwitchParameter(
   True=Multiply(
    A=@n44
    B=MaterialFunctionCall(Input=@n47.Result, Input Low=@n48, Input High=@n49) {materialFunction=&RemapValue_0-1}.Result
   )
   False=@n44
  ) {group=LitMat, parameterName=UseTexMask}
  n51 = Multiply(A=@n50, B=ScalarParameter() {defaultValue=0.2, group=LitMat, parameterName=EmissiveScale})
  n52 = DotProduct(A=SkyAtmosphereLightDirection(), B=@n25)
  n53 = VectorParameter() {group=Lighting, parameterName=Lighting_Offset/Hardness/Power, defaultValue=<rgba 0.1 1 2 0>}
  n54 = Saturate(None=Multiply(A=Add(A=@n52, B=@n53.R), B=@n53.G))
  n55 = Max(
   A=@n50
   B=LinearInterpolate(
    A=VectorParameter() {group=Lighting, parameterName=Lighting_ColorShadow, defaultValue=<rgba 0 0 0 1>}.RGB
    B=VectorParameter() {group=Lighting, parameterName=Lighting_ColorBright, defaultValue=<rgba 0.04 0.036 0.036 1>}.RGB
    Alpha=Power(Base=@n54, Exp=@n53.B)
   )
  )
  MaterialFunctionCall(Emissive=Multiply(
    A=MaterialFunctionCall(Color_Morning=@n4.RGB, Color_Day=@n5, Color_Evening=@n6.RGB, Color_Night=@n7.RGB) {materialFunction=&Color_Time_of_Day}.Result
    B=StaticSwitchParameter(True=@n51, False=@n55) {group=LitMat, parameterName=UseLitMat}
   )) {materialFunction=&MF_FX_EmissiveAdjustments}
  ScalarParameter() {defaultValue=0.5, group=LitMat, parameterName=Specular}
  Subtract(A=@n1.Max, B=@n1.Max)
 ]
}
